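# Conversation database for Season 26 Snapshot Part 04 main-questline NPC dialogue, covering quests Q03, Q05 and Q08.
# Each quest's entry tag leads to linear speech chains: Nolan Chance or Fish Thicc for Q03, Mae for Q05, Nolan for Q08.
# Opening lines auto-select when the quest is active, it and its first objective are incomplete, and the NPC carries the matching character tag.
# Every choice, including Continue, carries its quest's choice tag.
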ConversationDatabase /BattlePassS26_Quests/Conversations/Snapshots/Part04/Conversation_S26_Snapshot_P04
entry_points [
 {tag=Conversation.Entry.Season26.Snapshot.P04.Q03, destinations [#1]}
 {tag=Conversation.Entry.Season26.Snapshot.P04.Q08, destinations [#2]}
 {tag=Conversation.Entry.Season26.Snapshot.P04.Q05, destinations [#3]}
]
nodes [
 ConversationEntryPointNode_4 {
  id=#3
  class=ConversationEntryPointNode
  outputs [#4]
  properties {entryTag {tagName=Conversation.Entry.Season26.Snapshot.P04.Q05}}
 }
 Speech_C_8 {
  id=#4
  class=Speech_C
  outputs [#5]
  properties {
   softTaskIcon=&T_UI_NPC_MainQuest
   generalConfig {defaultMessage="You're formidable. And I love blowin' stuff up! :D We'd make a pretty great team... But just FYI..."}
  }
  sub_nodes [
   ConversationChoiceNode_46 {
    class=ConversationChoiceNode
    properties {
     defaultChoiceDisplayText=Talk to Mae
     choiceTags {gameplayTags [
       {tagName=Conversation.Choice.AutoSelectIfValid}
       {tagName=Conversation.Choice.Category.MainQuestline}
       {tagName=Athena.Quests.S26.Conversations.P04.Choice.Q05}
      ]}
    }
   }
   FortConversationRequirement_ControllerRequirement_88 {
    class=FortConversationRequirement_ControllerRequirement
    properties {
     nodeName=Require: Has Completed Quest
     requirement {
      class=FortControllerRequirement_HasCompletedQuest
      properties {bInverseCondition=true, questDefinition=/BattlePassS26_Quests/Items/QuestItems/Snapshot/Part04/quest_s26_snapshot_p04_q05}
     }
    }
   }
   FortConversationRequirement_ControllerRequirement_89 {
    class=FortConversationRequirement_ControllerRequirement
    properties {
     nodeName=Require: Is Quest Active
     requirement {
      class=FortControllerRequirement_IsQuestActive
      properties {questDefinition=&quest_s26_snapshot_p04_q05}
     }
    }
   }
   FortConversationRequirement_ControllerRequirement_90 {
    class=FortConversationRequirement_ControllerRequirement
    properties {
     nodeName=Require: Has Completed Quest Objective
     requirement {
      class=FortControllerRequirement_HasCompletedQuestObjective
      properties {
       bInverseCondition=true
       objectiveBackendName=quest_s26_snapshot_p04_q05_obj0
       questDefinition=&quest_s26_snapshot_p04_q05
      }
     }
    }
   }
   FortConversationRequirement_ControllerRequirement_91 {
    class=FortConversationRequirement_ControllerRequirement
    properties {
     nodeName=Require: NPC Has Conversation Tag
     participantId {tagName=Conversation.Participant.NPC}
     requirement {class=FortControllerRequirement_NPCHasConversationEntryTag, properties {queryTag {tagName=Conversation.Entry.IndieBucket}}}
    }
   }
  ]
 }
 Speech_C_9 {
  id=#5
  class=Speech_C
  outputs [#6]
  properties {generalConfig {defaultMessage=Kado's up to somethin suuuper sinister. I found some schematics lurking behind his firewall.}}
  sub_nodes [ConversationChoiceNode_14 {
    class=ConversationChoiceNode
    properties {defaultChoiceDisplayText=Continue, choiceTags {gameplayTags [{tagName=Athena.Quests.S26.Conversations.P04.Choice.Q05}]}}
   }]
 }
 Speech_C_18 {
  id=#6
  class=Speech_C
  outputs []
  properties {generalConfig {defaultMessage="Any idea what a Chrono-Cell is? I kinda wanna know before I, y'know, risk everything I have."}}
  sub_nodes [ConversationChoiceNode_47 {
    class=ConversationChoiceNode
    properties {defaultChoiceDisplayText=Continue, choiceTags {gameplayTags [{tagName=Athena.Quests.S26.Conversations.P04.Choice.Q05}]}}
   }]
 }
 ConversationEntryPointNode_3 {
  id=#2
  class=ConversationEntryPointNode
  outputs [#7]
  properties {entryTag {tagName=Conversation.Entry.Season26.Snapshot.P04.Q08}}
 }
 Speech_C_5 {
  id=#7
  class=Speech_C
  outputs [#8]
  properties {
   softTaskIcon=&T_UI_NPC_MainQuest
   generalConfig {defaultMessage="Mae's on board! That's good for the heist, and very very bad for my explosives budget..."}
  }
  sub_nodes [
   ConversationChoiceNode_5 {
    class=ConversationChoiceNode
    properties {
     defaultChoiceDisplayText=Talk to Nolan Chance
     choiceTags {gameplayTags [
       {tagName=Conversation.Choice.AutoSelectIfValid}
       {tagName=Conversation.Choice.Category.MainQuestline}
       {tagName=Athena.Quests.S26.Conversations.P04.Choice.Q08}
      ]}
    }
   }
   FortConversationRequirement_ControllerRequirement_28 {
    class=FortConversationRequirement_ControllerRequirement
    properties {
     nodeName=Require: Has Completed Quest
     requirement {
      class=FortControllerRequirement_HasCompletedQuest
      properties {bInverseCondition=true, questDefinition=/BattlePassS26_Quests/Items/QuestItems/Snapshot/Part04/quest_s26_snapshot_p04_q08}
     }
    }
   }
   FortConversationRequirement_ControllerRequirement_30 {
    class=FortConversationRequirement_ControllerRequirement
    properties {
     nodeName=Require: Is Quest Active
     requirement {
      class=FortControllerRequirement_IsQuestActive
      properties {questDefinition=&quest_s26_snapshot_p04_q08}
     }
    }
   }
   FortConversationRequirement_ControllerRequirement_31 {
    class=FortConversationRequirement_ControllerRequirement
    properties {
     nodeName=Require: Has Completed Quest Objective
     requirement {
      class=FortControllerRequirement_HasCompletedQuestObjective
      properties {
       bInverseCondition=true
       objectiveBackendName=quest_s26_snapshot_p04_q08_obj0
       questDefinition=&quest_s26_snapshot_p04_q08
      }
     }
    }
   }
   FortConversationRequirement_ControllerRequirement_32 {
    class=FortConversationRequirement_ControllerRequirement
    properties {
     nodeName=Require: NPC Has Conversation Tag
     participantId {tagName=Conversation.Participant.NPC}
     requirement {class=FortControllerRequirement_NPCHasConversationEntryTag, properties {queryTag {tagName=Conversation.Entry.MusketSlinger}}}
    }
   }
  ]
 }
 Speech_C_13 {
  id=#8
  class=Speech_C
  outputs [#9]
  properties {generalConfig {defaultMessage="Honestly, I don't know if we could have pulled it off without her. Great work."}}
  sub_nodes [ConversationChoiceNode_0 {
    class=ConversationChoiceNode
    properties {defaultChoiceDisplayText=Continue, choiceTags {gameplayTags [{tagName=Athena.Quests.S26.Conversations.P04.Choice.Q08}]}}
   }]
 }
 Speech_C_0 {
  id=#9
  class=Speech_C
  outputs []
  properties {generalConfig {defaultMessage="We're almost there. Once we deliver what Slone is after, you and I are gonna be walkin' easy street!"}}
  sub_nodes [ConversationChoiceNode_9 {
    class=ConversationChoiceNode
    properties {defaultChoiceDisplayText=Continue, choiceTags {gameplayTags [{tagName=Athena.Quests.S26.Conversations.P04.Choice.Q08}]}}
   }]
 }
 ConversationEntryPointNode_1 {
  id=#1
  class=ConversationEntryPointNode
  outputs [#10, #11]
  properties {entryTag {tagName=Conversation.Entry.Season26.Snapshot.P04.Q03}}
 }
 Speech_C_3 {
  id=#11
  class=Speech_C
  outputs [#12]
  properties {
   softTaskIcon=&T_UI_NPC_MainQuest
   generalConfig {defaultMessage=Mae's been my best fin since we were guppies. She'll only work with people who impress her.}
  }
  sub_nodes [
   ConversationChoiceNode_36 {
    class=ConversationChoiceNode
    properties {
     defaultChoiceDisplayText=Talk to Fish Thicc
     choiceTags {gameplayTags [
       {tagName=Conversation.Choice.AutoSelectIfValid}
       {tagName=Conversation.Choice.Category.MainQuestline}
       {tagName=Athena.Quests.S26.Conversations.P04.Choice.Q03}
      ]}
    }
   }
   FortConversationRequirement_ControllerRequirement_68 {
    class=FortConversationRequirement_ControllerRequirement
    properties {
     nodeName=Require: Has Completed Quest
     requirement {
      class=FortControllerRequirement_HasCompletedQuest
      properties {bInverseCondition=true, questDefinition=/BattlePassS26_Quests/Items/QuestItems/Snapshot/Part04/quest_s26_snapshot_p04_q03}
     }
    }
   }
   FortConversationRequirement_ControllerRequirement_69 {
    class=FortConversationRequirement_ControllerRequirement
    properties {
     nodeName=Require: Is Quest Active
     requirement {
      class=FortControllerRequirement_IsQuestActive
      properties {questDefinition=&quest_s26_snapshot_p04_q03}
     }
    }
   }
   FortConversationRequirement_ControllerRequirement_70 {
    class=FortConversationRequirement_ControllerRequirement
    properties {
     nodeName=Require: Has Completed Quest Objective
     requirement {
      class=FortControllerRequirement_HasCompletedQuestObjective
      properties {
       bInverseCondition=true
       objectiveBackendName=quest_s26_snapshot_p04_q03_obj0
       questDefinition=&quest_s26_snapshot_p04_q03
      }
     }
    }
   }
   FortConversationRequirement_ControllerRequirement_71 {
    class=FortConversationRequirement_ControllerRequirement
    properties {
     nodeName=Require: NPC Has Conversation Tag
     participantId {tagName=Conversation.Participant.NPC}
     requirement {class=FortControllerRequirement_NPCHasConversationEntryTag, properties {queryTag {tagName=Conversation.Entry.FishThicc}}}
    }
   }
  ]
 }
 Speech_C_4 {
  id=#12
  class=Speech_C
  outputs []
  properties {generalConfig {defaultMessage="The best way to impress Mae? Hackin', survivin', and stealin' of course!"}}
  sub_nodes [ConversationChoiceNode_37 {
    class=ConversationChoiceNode
    properties {defaultChoiceDisplayText=Continue, choiceTags {gameplayTags [{tagName=Athena.Quests.S26.Conversations.P04.Choice.Q03}]}}
   }]
 }
 Speech_C_1 {
  id=#10
  class=Speech_C
  outputs [#13]
  properties {
   softTaskIcon=&T_UI_NPC_MainQuest
   generalConfig {defaultMessage=I spoke to Fish Thicc. He's Mae's best friend. Pfft. Not like I wanted a best friend or anything...}
  }
  sub_nodes [
   ConversationChoiceNode_1 {
    class=ConversationChoiceNode
    properties {
     defaultChoiceDisplayText=Talk to Nolan Chance
     choiceTags {gameplayTags [
       {tagName=Conversation.Choice.AutoSelectIfValid}
       {tagName=Conversation.Choice.Category.MainQuestline}
       {tagName=Athena.Quests.S26.Conversations.P04.Choice.Q03}
      ]}
    }
   }
   FortConversationRequirement_ControllerRequirement_0 {
    class=FortConversationRequirement_ControllerRequirement
    properties {
     nodeName=Require: Has Completed Quest
     requirement {
      class=FortControllerRequirement_HasCompletedQuest
      properties {bInverseCondition=true, questDefinition=&quest_s26_snapshot_p04_q03}
     }
    }
   }
   FortConversationRequirement_ControllerRequirement_5 {
    class=FortConversationRequirement_ControllerRequirement
    properties {
     nodeName=Require: Is Quest Active
     requirement {
      class=FortControllerRequirement_IsQuestActive
      properties {questDefinition=&quest_s26_snapshot_p04_q03}
     }
    }
   }
   FortConversationRequirement_ControllerRequirement_6 {
    class=FortConversationRequirement_ControllerRequirement
    properties {
     nodeName=Require: Has Completed Quest Objective
     requirement {
      class=FortControllerRequirement_HasCompletedQuestObjective
      properties {
       bInverseCondition=true
       objectiveBackendName=quest_s26_snapshot_p04_q03_obj0
       questDefinition=&quest_s26_snapshot_p04_q03
      }
     }
    }
   }
   FortConversationRequirement_ControllerRequirement_7 {
    class=FortConversationRequirement_ControllerRequirement
    properties {
     nodeName=Require: NPC Has Conversation Tag
     participantId {tagName=Conversation.Participant.NPC}
     requirement {class=FortControllerRequirement_NPCHasConversationEntryTag, properties {queryTag {tagName=Conversation.Entry.MusketSlinger}}}
    }
   }
  ]
 }
 Speech_C_2 {
  id=#13
  class=Speech_C
  outputs [#14]
  properties {generalConfig {defaultMessage=You're gonna have to get a little reckless to catch her attention.}}
  sub_nodes [ConversationChoiceNode_13 {
    class=ConversationChoiceNode
    properties {defaultChoiceDisplayText=Continue, choiceTags {gameplayTags [{tagName=Athena.Quests.S26.Conversations.P04.Choice.Q03}]}}
   }]
 }
 Speech_C_17 {
  id=#14
  class=Speech_C
  outputs []
  properties {generalConfig {defaultMessage="But you do reckless very, very well. Right best friend? Or just... Regular friend? Sorry, I was trying it on."}}
  sub_nodes [ConversationChoiceNode_2 {
    class=ConversationChoiceNode
    properties {defaultChoiceDisplayText=Continue, choiceTags {gameplayTags [{tagName=Athena.Quests.S26.Conversations.P04.Choice.Q03}]}}
   }]
 }
]
